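# Pilgrim player avatar icon widget that shows a player's social avatar over their identity color.
# PreConstruct applies the size override and picks a square or default avatar material, the latter with optional border.
# SetPlayerImage takes a player state, then writes the background color and async-loaded avatar texture into the image's dynamic material.
# An invalid avatar reference triggers a countdown of one-second retries, printing an error when exhausted.
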WidgetBlueprint /PilgrimCore/Widgets/Player/WBP_Pilgrim_AvatarIcon
graphs [
 EventGraph {nodes [
   n1 = SetPlayerImage!() -> (OutputDelegate, then, FortPlayerStateAthena:Fort Player State Athena Object Reference) {is=title K2Node_CustomEvent}
   > SetPlayerAvatarBackground!(
    execute
    self
    BgColor:Linear Color Structure=GetPilgrim PlayerIdentity Color By PlayerState!(Player State:Player State Object Reference=@n1.FortPlayerStateAthena) -> (Player Identity Color:Linear Color Structure)
   ) -> (then)
   > n2 = Set PlayerIconRetryCount(PlayerIconRetryCount:int=Get PlayerIconRetryCount_Default() -> (PlayerIconRetryCount_Default:int=0)) -> (Output_Get:int=0)
   n3 = GetSocialAvatarBrush!(self:Fort Player State Zone Object Reference=@n1.FortPlayerStateAthena, bSmallImage:bool=true) -> (ReturnValue:Texture 2D Soft Object Reference)
   n4 = Branch(execute=[@n8.then, @n2.then], Condition=IsValidSoftObjectReference(SoftObjectReference=@n3))
   > n5 = Async Load Asset(Asset=@n3)
   n6 = Cast To Texture2D(execute=@n5.Completed, Object=@n5)
   > SetPlayerAvatarIcon!(execute, self, Icon:Texture Object Reference=@n6) -> (then)
   n7 = Branch(execute=@n4.else, Condition="integer > integer"(A=Get PlayerIconRetryCount() -> (PlayerIconRetryCount:int=0)))
   > Decrement Int(Value=Get PlayerIconRetryCount() -> (PlayerIconRetryCount=0))
   > n8 = Delay(Duration=1)
   PrintString!(execute=@n7.else, InString=Failed to get a vaild texture reference for player, bPrintToScreen, bPrintToLog, TextColor, Duration, Key) -> (then)
   Event PreConstruct()
   > OverridePlayerAvatarBrushSize!(execute, self, SizeOverride:Vector 2D Structure=Get PlayerIconSizeOverride() -> (PlayerIconSizeOverride:Vector 2D Structure)) -> (then)
   > n9 = Branch(Condition=Get bIsUsingSquareIcon() -> (bIsUsingSquareIcon:bool))
   > SetBrushFromSoftMaterial!(
    execute
    self:Image Object Reference=Get PlayerImage() -> (PlayerImage:Image Object Reference)
    SoftMaterial:Material Interface Soft Object Reference=/SparksUICommon/Materials/M_UI_BandMember_Avatar_SquareRounded
   ) -> (then)
   SetBrushFromSoftMaterial(
    execute=@n9.else
    self=Get PlayerImage()
    SoftMaterial=Select(
     Option 0:Material Interface Soft Object Reference=/SparksUICommon/Materials/MI_UI_BandMember_Avatar_Master_LocalPlayerWithouBorder
     Option 1:Material Interface Soft Object Reference=/SparksUICommon/Materials/MI_UI_BandMember_Avatar
     Index=Get ShowBorder() -> (ShowBorder:bool)
    ) -> (ReturnValue:Material Interface Soft Object Reference)
   )
   Reroute Node(InputPin:Wildcard) -> (OutputPin:Wildcard)
  ]}
 OverridePlayerAvatarBrushSize {nodes [
   n1 = OverridePlayerAvatarBrushSize!() -> (then, SizeOverride:Vector 2D Structure) {is=title K2Node_FunctionEntry}
   > SetDesiredSizeOverride!(execute, self:Image Object Reference=Get PlayerImage(), DesiredSize:Vector 2D Structure=@n1) -> (then)
  ]}
 SetPlayerAvatarBackground {nodes [
   n1 = SetPlayerAvatarBackground!() -> (then, BgColor:Linear Color Structure) {is=title K2Node_FunctionEntry}
   > n2 = GetDynamicMaterial(self=Get PlayerImage())
   > SetVectorParameterValue(self=@n2, ParameterName=BgColor, Value=@n1)
  ]}
 SetPlayerAvatarIcon {nodes [
   n1 = SetPlayerAvatarIcon!() -> (then, Icon:Texture Object Reference) {is=title K2Node_FunctionEntry}
   > n2 = GetDynamicMaterial(self=Get PlayerImage())
   > SetTextureParameterValue(self=@n2, ParameterName=PlayerIcon, Value=@n1)
  ]}
]
variables [
 bIsUsingSquareIcon {type=boolean}
 PlayerIconRetryCount {type=integer}
 PlayerIconRetryCount_Default {type=integer, default=10}
 PlayerIconSizeOverride {type=Vector2D, default=<32 32>}
 ShowBorder {type=boolean, default=true}
]
widgets {
 info {inheritedWidgetCount=0, parentClass=&UserWidget, rootWidgetClass=&Overlay, widgetCount=2}
 widgets [
  Overlay_26 {widgetClassPath=&Overlay, widget=Overlay_26}
  PlayerImage {
   widgetClassPath=&Image
   bIsVariable=true
   parent=Overlay_26
   widget=PlayerImage
   properties {brush {
     bIsDynamicallyLoaded=false
     drawAs=Image
     imageType=FullColor
     mirroring=NoMirror
     resourceName=None
     resourceObject=&MI_UI_BandMember_Avatar
     tiling=NoTile
     imageSize=<32 32>
     margin=<ltrb 0 0 0 0>
     outlineSettings {
      bUseBrushTransparency=false
      roundingType=HalfHeightRadius
      width=0
      color {colorUseRule=UseColor_Specified, specifiedColor=<rgba 0 0 0 0>}
      cornerRadii=<0 0 0 0>
     }
     tintColor {colorUseRule=UseColor_Specified, specifiedColor=<rgba 1 1 1 1>}
     uVRegion {bIsValid=false, max=<0 0>, min=<0 0>}
    }}
   slot {class=OverlaySlot, properties {horizontalAlignment=HAlign_Center, verticalAlignment=VAlign_Center}}
  }
 ]
}
