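# WBP_Pilgrim_BandLeaderIcon is a small Pilgrim UI widget that displays the band leader icon.
# Its only element is a lazy-loading image using a local-multiplayer lead icon material.
# On PreConstruct, it passes the IconSizeOverride variable to OverrideIconSize, which sets the image's desired size override in place of the brush's own size.
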WidgetBlueprint /PilgrimCore/Widgets/Player/WBP_Pilgrim_BandLeaderIcon
graphs [
 EventGraph {nodes [
   Event PreConstruct()
   > OverrideIconSize!(execute, self, SizeOverride:Vector 2D Structure=Get IconSizeOverride() -> (IconSizeOverride:Vector 2D Structure)) -> (then)
  ]}
 OverrideIconSize {nodes [
   n1 = OverrideIconSize!() -> (then, SizeOverride:Vector 2D Structure) {is=title K2Node_FunctionEntry}
   > SetDesiredSizeOverride!(
    execute
    self:Image Object Reference=Get BandLeaderIcon() -> (BandLeaderIcon:Fort Lazy Image Object Reference)
    DesiredSize:Vector 2D Structure=@n1
   ) -> (then)
  ]}
]
variables [IconSizeOverride {type=Vector2D, default=<32 32>}]
widgets {
 info {inheritedWidgetCount=0, parentClass=&UserWidget, rootWidgetClass=&FortLazyImage, widgetCount=1}
 widgets [BandLeaderIcon {
   widgetClassPath=&FortLazyImage
   bIsVariable=true
   widget=BandLeaderIcon
   properties {brush {
     bIsDynamicallyLoaded=false
     drawAs=Image
     imageType=FullColor
     mirroring=NoMirror
     resourceName=None
     resourceObject=/PilgrimCore/Widgets/Materials/LocalPlayerMaterials/MI_Ui_LocalMultiplayer_Icon_Lead
     tiling=NoTile
     imageSize=<104 104>
     margin=<ltrb 0 0 0 0>
     outlineSettings {
      bUseBrushTransparency=false
      roundingType=HalfHeightRadius
      width=0
      color {colorUseRule=UseColor_Specified, specifiedColor=<rgba 0 0 0 0>}
      cornerRadii=<0 0 0 0>
     }
     tintColor {colorUseRule=UseColor_Specified, specifiedColor=<rgba 1 1 1 1>}
     uVRegion {bIsValid=false, max=<0 0>, min=<0 0>}
    }}
  }]
}
