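# WBP_Pilgrim_BandLeaderIcon_WithVM is a wrapper user widget around the Pilgrim band leader icon, used among the player widgets.
# Its only widget, at the root, is a WBP_Pilgrim_BandLeaderIcon instance.
# PreConstruct passes the IconSizeOverride variable through a local OverrideIconSize function to the inner icon's function of the same name.
# Despite the WithVM name, no view model or binding appears here.
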WidgetBlueprint /PilgrimCore/Widgets/Player/WBP_Pilgrim_BandLeaderIcon_WithVM
graphs [
 EventGraph {nodes [
   Event PreConstruct()
   > OverrideIconSize!(execute, self, SizeOverride:Vector 2D Structure=Get IconSizeOverride() -> (IconSizeOverride:Vector 2D Structure)) -> (then)
  ]}
 OverrideIconSize {nodes [
   n1 = OverrideIconSize!() -> (then, SizeOverride:Vector 2D Structure) {is=title K2Node_FunctionEntry}
   > OverrideIconSize(
    self:WBP Pilgrim Band Leader Icon Object Reference=Get BandLeaderIcon() -> (BandLeaderIcon:WBP Pilgrim Band Leader Icon Object Reference)
    SizeOverride=@n1
   ) {is=title K2Node_CallFunction}
  ]}
]
variables [IconSizeOverride {type=Vector2D, default=<32 32>}]
widgets {
 info {
  inheritedWidgetCount=0
  parentClass=&UserWidget
  rootWidgetClass=&/WBP_Pilgrim_BandLeaderIcon.WBP_Pilgrim_BandLeaderIcon_C
  widgetCount=1
 }
 widgets [BandLeaderIcon {widgetClassPath=&/WBP_Pilgrim_BandLeaderIcon, bIsVariable=true, widget=BandLeaderIcon}]
}
