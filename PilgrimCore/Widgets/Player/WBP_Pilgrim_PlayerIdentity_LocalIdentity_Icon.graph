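# Widget Blueprint for a single icon that marks a local splitscreen player's identity by color in the Pilgrim UI.
# PreConstruct applies the icon size override and previews a fixed index at design time.
# SetLocalPlayerIndex tints the icon for any index except -1, fed directly or from a PlayerState's splitscreen index.
# Its per-index texture loop has no incoming execution, and every array default is the same diamond.
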WidgetBlueprint /PilgrimCore/Widgets/Player/WBP_Pilgrim_PlayerIdentity_LocalIdentity_Icon
graphs [
 EventGraph {nodes [
   n1 = Event PreConstruct()
   > n2 = Sequence()
   > SetDesiredSizeOverride!(
    execute
    self:Image Object Reference=Get PlayerName_Icon() -> (PlayerName_Icon:Fort Lazy Image Object Reference)
    DesiredSize:Vector 2D Structure=Get IconSizeOverride() -> (IconSizeOverride:Vector 2D Structure)
   ) -> (then)
   Branch(execute=@n2.then_1, Condition=@n1.IsDesignTime)
   > SetLocalPlayerIndex!(execute, self, LocalPlayerIndex:int=2) -> (then)
  ]}
 SetLocalPlayerIndex {nodes [
   n1 = SetLocalPlayerIndex!() -> (then, LocalPlayerIndex:int) {is=title K2Node_FunctionEntry}
   > n2 = Branch(Condition="Not Equal (Integer)"(A=@n1, B=-1))
   n3 = Get PlayerName_Icon()
   SetColorAndOpacity(
    execute=[@n2.then, @n4.Completed]
    self:Image Object Reference=@n3
    InColorAndOpacity:Linear Color Structure=GetPilgrim LocalPlayerIdentity Color By Index!(LocalPlayerIndex:int=@n1) -> (Player Identity Color:Linear Color Structure)
   )
   n4 = For Each Loop(Array:Array of Texture 2D Soft Object References=Get PlayerIDIcons() -> (PlayerIDIcons:Array of Texture 2D Soft Object References)) -> (Array Element:Texture 2D Soft Object Reference)
   > Branch(Condition="Equal (Integer)"(A=@n4.Array Index, B=@n1))
   > SetBrushFromLazyTexture!(execute, self:Common Lazy Image Object Reference=@n3, LazyTexture:"Texture 2D Soft Object Reference (by ref)"=@n4, bMatchSize:bool=false) -> (then)
  ]}
 SetLocalPlayerIndexFromPlayerState {nodes [
   n1 = SetLocalPlayerIndexFromPlayerState!() -> (then, Player State:Player State Object Reference) {is=title K2Node_FunctionEntry}
   > SetLocalPlayerIndex(LocalPlayerIndex=GetSplitscreenIndex!(PC:Player Controller Object Reference=GetFortPlayerControllerFromActor!(Actor:Actor Object Reference=@n1) -> (ReturnValue:Fort Player Controller Object Reference)) -> (ReturnValue:int=0)) {is=title K2Node_CallFunction}
  ]}
]
variables [
 IconSizeOverride {type=Vector2D, default=<14 14>}
 PlayerIDIcons {
  type="array<Texture2D>"
  default [
   /Game/Athena/HUD/Art/T_UI_Indicator_Diamond
   &T_UI_Indicator_Diamond
   &T_UI_Indicator_Diamond
   &T_UI_Indicator_Diamond
  ]
 }
]
widgets {
 info {inheritedWidgetCount=0, parentClass=&UserWidget, rootWidgetClass=&FortLazyImage, widgetCount=1}
 widgets [PlayerName_Icon {
   widgetClassPath=&FortLazyImage
   bIsVariable=true
   widget=PlayerName_Icon
   properties {brush {
     bIsDynamicallyLoaded=false
     drawAs=Image
     imageType=FullColor
     mirroring=NoMirror
     resourceName=None
     resourceObject=&T_UI_Indicator_Diamond
     tiling=NoTile
     imageSize=<64 64>
     margin=<ltrb 0 0 0 0>
     outlineSettings {
      bUseBrushTransparency=false
      roundingType=HalfHeightRadius
      width=0
      color {colorUseRule=UseColor_Specified, specifiedColor=<rgba 0 0 0 0>}
      cornerRadii=<0 0 0 0>
     }
     tintColor {colorUseRule=UseColor_Specified, specifiedColor=<rgba 1 1 1 1>}
     uVRegion {bIsValid=false, max=<0 0>, min=<0 0>}
    }}
  }]
}
