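# Text label widget that shows a local splitscreen player's identity name in that player's color.
# SetLocalPlayerIndex looks up the Pilgrim identity definition by index, formats its label through TextFormatOverride and tints the text, skipping index -1.
# SetLocalPlayerIndexFromPlayerState derives that index from the player state's Fortnite controller splitscreen index.
# PreConstruct applies FontSizeOverride to the font and, at design time, previews a fixed sample player.
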WidgetBlueprint /PilgrimCore/Widgets/Player/WBP_Pilgrim_PlayerIdentity_LocalIdentity_Label
graphs [
 EventGraph {nodes [
   n1 = Event PreConstruct()
   > n2 = Sequence()
   n3 = Get PlayerName_Label() -> (PlayerName_Label:Text Block Object Reference)
   > n4 = Set members in Slate Font Info!(
    execute:Exec
    StructRef:"Slate Font Info Structure (by ref)"=Get Font!(self:Text Object Reference=@n3) -> (Font:Slate Font Info Structure)
    Size:float=Get FontSizeOverride() -> (FontSizeOverride:double)
   ) -> (then:Exec, StructOut:"Slate Font Info Structure (by ref)") {is=title K2Node_SetFieldsInStruct}
   > SetFont(self=@n3, InFontInfo=@n4)
   Branch(execute=@n2.then_1, Condition=@n1.IsDesignTime)
   > SetLocalPlayerIndex!(execute, self, LocalPlayerIndex:int=2) -> (then)
  ]}
 SetLocalPlayerIndex {nodes [
   n1 = SetLocalPlayerIndex!() -> (then, LocalPlayerIndex:int) {is=title K2Node_FunctionEntry}
   > Branch(Condition="Not Equal (Integer)"(A=@n1, B=-1))
   n2 = Break S Pilgrim Local Player Identifier Definition!(S_Pilgrim_LocalPlayerIdentifierDefinition:"S Pilgrim Local Player Identifier Definition Structure (by ref)"=GetPilgrim LocalPlayerIdentity ColorInfo By Index!(LocalPlayerIndex:int=@n1) -> (Player Identity Color Info:S Pilgrim Local Player Identifier Definition Structure)) -> (IdentityLabel_16_#1:Text, IdentityColor_5_#2:Linear Color Structure) {is=title K2Node_BreakStruct}
   n3 = Get PlayerName_Label()
   > "SetText (Text)"(self=@n3, InText=Format Text!(Format=Get TextFormatOverride() -> (TextFormatOverride:Text), PlayerID:Text=@n2) -> (Result))
   > SetColorAndOpacity(
    self=@n3
    InColorAndOpacity="To SlateColor (Color)"!(Value:Linear Color Structure=@n2.IdentityColor_5_#2) -> (ReturnValue:Slate Color Structure)
   )
  ]}
 SetLocalPlayerIndexFromPlayerState {nodes [
   n1 = SetLocalPlayerIndexFromPlayerState!() -> (then, Player State:Player State Object Reference) {is=title K2Node_FunctionEntry}
   > SetLocalPlayerIndex(LocalPlayerIndex=GetSplitscreenIndex!(PC:Player Controller Object Reference=GetFortPlayerControllerFromActor!(Actor:Actor Object Reference=@n1) -> (ReturnValue:Fort Player Controller Object Reference)) -> (ReturnValue:int=0)) {is=title K2Node_CallFunction}
  ]}
]
variables [
 FontSizeOverride {type=number, default=54}
 TextFormatOverride {type=string, default="{PlayerID}"}
]
widgets {
 info {
  inheritedWidgetCount=0
  parentClass=&UserWidget
  rootWidgetClass=/Game/Valkyrie/UMG/UEFN_TextBlock.UEFN_TextBlock_C
  widgetCount=1
 }
 widgets [PlayerName_Label {
   widgetClassPath=&UEFN_TextBlock
   bIsVariable=true
   widget=PlayerName_Label
   properties {
    text=SetByBP
    visibility=HitTestInvisible
    font {
     bForceMonospaced=false
     bMaterialIsStencil=false
     fontMaterial=null
     fontObject=&HeadingNow
     letterSpacing=0
     monospacedWidth=1
     size=40.5
     skewAmount=0
     typefaceFontName=75 Medium
     outlineSettings {
      bApplyOutlineToDropShadows=false
      bMiteredCorners=false
      bSeparateFillAlpha=false
      outlineMaterial=null
      outlineSize=0
      outlineColor=<rgba 0 0 0 1>
     }
    }
   }
  }]
}
