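# Pill widget that identifies a player with avatar, local identity label and icon, and band leader icon over a player-colored underlayer.
# SetupPlayerIdentity shows it only during an active HamSandwich session or when forced, passing the given or owning PlayerState to OverridePlayerInfo.
# OverridePlayerInfo derives color, identity and avatar from that state, showing the leader icon for the primary player.
# PreConstruct sets the avatar material and, at design time, previews a configurable local player index.
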WidgetBlueprint /PilgrimCore/Widgets/Player/WBP_Pilgrim_PlayerIdentity_Pill
graphs [
 EventGraph {nodes [
   n1 = Event PreConstruct()
   > SetBrushFromMaterial(
    self=Get PlayerImage!(self:WBP Pilgrim Avatar Icon Object Reference=Get AvatarIcon() -> (AvatarIcon:WBP Pilgrim Avatar Icon Object Reference)) -> (PlayerImage:Image Object Reference)
    Material=/SparksUICommon/Materials/MI_UI_BandMember_Avatar_Master_LocalPlayerWithouBorder
   )
   > Branch(Condition=@n1.IsDesignTime)
   n2 = Get Design Preview Local Player Index() -> (Design Preview Local Player Index:int=0)
   n3 = GetPilgrim LocalPlayerIdentity ColorInfo By Index!(LocalPlayerIndex:int=@n2) -> (Player Identity Color Info:S Pilgrim Local Player Identifier Definition Structure)
   > SetPlayerColorInfo!(execute, self, LocalidColor:S Pilgrim Local Player Identifier Definition Structure=@n3) -> (then)
   n4 = Break S Pilgrim Local Player Identifier Definition!(S_Pilgrim_LocalPlayerIdentifierDefinition:"S Pilgrim Local Player Identifier Definition Structure (by ref)"=@n3) -> (
    IdentityColor_5_#1:Linear Color Structure
    PlayerColorUnderlayerM_11_#2:Material Instance Object Reference
   ) {is=title K2Node_BreakStruct}
   > SetPlayerAvatarBackground!(execute, self:WBP Pilgrim Avatar Icon Object Reference=Get AvatarIcon(), BgColor:Linear Color Structure=@n4) -> (then)
   > SetBrushFromMaterial(
    self=Get PlayerUnderlayer() -> (PlayerUnderlayer:Image Object Reference)
    Material=@n4.PlayerColorUnderlayerM_11_#2
   )
   > SetLocalPlayerIndex!(
    execute
    self:WBP Pilgrim Player Identity Local Identity Label Object Reference=Get PlayerIdentity_LocalIdentity_Label() -> (PlayerIdentity_LocalIdentity_Label:WBP Pilgrim Player Identity Local Identity Label Object Reference)
    LocalPlayerIndex:int=@n2
   ) -> (then)
   > SetLocalPlayerIndex(
    self:WBP Pilgrim Player Identity Local Identity Icon Object Reference=Get PlayerIdentity_LocalIdentity_Icon() -> (PlayerIdentity_LocalIdentity_Icon:WBP Pilgrim Player Identity Local Identity Icon Object Reference)
    LocalPlayerIndex=@n2
   )
   > SetVisibility(
    self=Get BandLeaderIcon() -> (BandLeaderIcon:WBP Pilgrim Band Leader Icon With VM Object Reference)
    InVisibility=Select(Option 0=Collapsed, Option 1=SelfHitTestInvisible, Index="Equal (Integer)"(A=@n2))
   )
  ]}
 HidePlayerDisplay {nodes [
   HidePlayerDisplay() {is=title K2Node_FunctionEntry}
   > SetVisibility(self=Self-Reference(), InVisibility=Hidden)
  ]}
 OverridePlayerInfo {nodes [
   n1 = OverridePlayerInfo!() -> (then, PlayerState:Player State Object Reference) {is=title K2Node_FunctionEntry}
   > SetPlayerColorInfo(LocalidColor=GetPilgrim PlayerIdentity ColorInfo By PlayerState!(Player State:Player State Object Reference=@n1) -> (Player Identity Color Info:S Pilgrim Local Player Identifier Definition Structure))
   > SetLocalPlayerIndexFromPlayerState!(
    execute
    self:WBP Pilgrim Player Identity Local Identity Label Object Reference=Get PlayerIdentity_LocalIdentity_Label()
    Player State:Player State Object Reference=@n1
   ) -> (then)
   > SetLocalPlayerIndexFromPlayerState(self:WBP Pilgrim Player Identity Local Identity Icon Object Reference=Get PlayerIdentity_LocalIdentity_Icon(), Player State=@n1)
   n2 = GetPlayerController!(self:Player State Object Reference=@n1) -> (ReturnValue:Player Controller Object Reference)
   > SetVisibility(
    self=Get BandLeaderIcon()
    InVisibility=Select(Option 0=Collapsed, Option 1=SelfHitTestInvisible, Index=IsPrimaryPlayer!(self:Player Controller Object Reference=@n2) -> (ReturnValue:bool))
   )
   > n3 = Cast To FortPlayerStateAthena!(execute, Object=@n1) -> (then, CastFailed, AsFort Player State Athena:Fort Player State Athena Object Reference) {is=title K2Node_DynamicCast}
   > SetPlayerImage!(execute, self:WBP Pilgrim Avatar Icon Object Reference=Get AvatarIcon(), FortPlayerStateAthena:Fort Player State Athena Object Reference=@n3) -> (then)
  ]}
 SetPlayerColorInfo {nodes [
   n1 = SetPlayerColorInfo!() -> (then, LocalidColor:S Pilgrim Local Player Identifier Definition Structure) {is=title K2Node_FunctionEntry}
   > SetBrushFromMaterial(
    self=Get PlayerUnderlayer()
    Material=Break S Pilgrim Local Player Identifier Definition!(S_Pilgrim_LocalPlayerIdentifierDefinition=@n1) -> (PlayerColorUnderlayerM_11_#2)
   )
  ]}
 SetupPlayerIdentity {nodes [
   n1 = SetupPlayerIdentity!() -> (then, OptionalPlayerState:Player State Object Reference, ForceVisible:bool=~) {is=title K2Node_FunctionEntry}
   n2 = OR Boolean(A=IsHamSandwichSessionActive!() -> (ReturnValue:bool), B=@n1.ForceVisible)
   > n3 = Branch(Condition=@n2)
   n4 = GetOwningPlayer()
   > n5 = OverridePlayerInfo!(
    execute
    self
    PlayerState:Player State Object Reference=Select(
     Option 0:Player State Object Reference=Get PlayerState!(self:Controller Object Reference=@n4) -> (PlayerState:Player State Object Reference)
     Option 1:Player State Object Reference=@n1
     Index=IsValid(Object=@n1)
    ) -> (ReturnValue:Player State Object Reference)
   ) -> (then) {is=title K2Node_CallFunction}
   SetVisibility(
    execute=[@n5.then, @n3.else]
    InVisibility=Select(Option 0:ESlateVisibility Enum=Hidden, Option 1:ESlateVisibility Enum=SelfHitTestInvisible, Index=@n2) -> (ReturnValue:ESlateVisibility Enum)
   )
   > Return Node()
  ]}
 ShowPlayerDisplay {nodes [
   ShowPlayerDisplay() {is=title K2Node_FunctionEntry}
   > SetVisibility(self=Self-Reference(), InVisibility=SelfHitTestInvisible)
  ]}
]
variables [
 Design Preview Local Player Index {type=integer, default=2}
 MyPlayerState {type=PlayerState, default=null}
 PlayerIconRetryCount {type=integer, default=10}
 PlayerIconRetryCount_Default {type=integer, default=10}
]
widgets {
 info {inheritedWidgetCount=0, parentClass=&UserWidget, rootWidgetClass=&GridPanel, widgetCount=9}
 widgets [
  GridPanel_0 {widgetClassPath=&GridPanel, widget=GridPanel_0, properties {columnFill [1], rowFill [1]}}
  PlayerUnderlayer {
   widgetClassPath=&Image
   bIsVariable=true
   parent=GridPanel_0
   widget=PlayerUnderlayer
   properties {brush {
     bIsDynamicallyLoaded=false
     drawAs=Image
     imageType=FullColor
     mirroring=NoMirror
     resourceName=None
     resourceObject=/PilgrimCore/Widgets/Materials/LocalPlayerMaterials/M_PilgrimLocalPlayer_Underlayer_P3
     tiling=NoTile
     imageSize=<32 32>
     margin=<ltrb 0 0 0 0>
     outlineSettings {
      bUseBrushTransparency=false
      roundingType=HalfHeightRadius
      width=0
      color {colorUseRule=UseColor_Specified, specifiedColor=<rgba 0 0 0 0>}
      cornerRadii=<0 0 0 0>
     }
     tintColor {colorUseRule=UseColor_Specified, specifiedColor=<rgba 1 1 1 1>}
     uVRegion {bIsValid=false, max=<0 0>, min=<0 0>}
    }}
   slot {class=GridSlot, properties {columnSpan=1, rowSpan=1}}
  }
  PlayerID {
   widgetClassPath=&HorizontalBox
   parent=GridPanel_0
   widget=PlayerID
   slot {class=GridSlot, properties {columnSpan=1, layer=10, rowSpan=1, padding=<ltrb 0 0 16 0>}}
  }
  AvatarIcon {
   widgetClassPath=&/WBP_Pilgrim_AvatarIcon
   bIsVariable=true
   parent=PlayerID
   widget=AvatarIcon
   properties {playerIconSizeOverride=<46 46>}
   slot {class=HorizontalBoxSlot, properties {padding=<ltrb 3 3 0 3>}}
  }
  Spacer_62 {widgetClassPath=&Spacer, parent=PlayerID, widget=Spacer_62, properties {size=<6 1>}, slot {class=HorizontalBoxSlot}}
  PlayerIdentity_LocalIdentity_Label {
   widgetClassPath=&/WBP_Pilgrim_PlayerIdentity_LocalIdentity_Label
   bIsVariable=true
   parent=PlayerID
   widget=PlayerIdentity_LocalIdentity_Label
   properties {fontSizeOverride=18}
   slot {class=HorizontalBoxSlot, properties {verticalAlignment=VAlign_Center, padding=<ltrb 0 0 2 0>}}
  }
  PlayerIdentity_LocalIdentity_Icon {
   widgetClassPath=&/WBP_Pilgrim_PlayerIdentity_LocalIdentity_Icon
   bIsVariable=true
   parent=PlayerID
   widget=PlayerIdentity_LocalIdentity_Icon
   properties {visibility=Collapsed, iconSizeOverride=<18 18>}
   slot {class=HorizontalBoxSlot, properties {verticalAlignment=VAlign_Center}}
  }
  Spacer_1 {widgetClassPath=&Spacer, parent=PlayerID, widget=Spacer_1, properties {size=<2 0>}, slot {class=HorizontalBoxSlot}}
  BandLeaderIcon {
   widgetClassPath=&/WBP_Pilgrim_BandLeaderIcon_WithVM
   bIsVariable=true
   parent=PlayerID
   widget=BandLeaderIcon
   slot {class=HorizontalBoxSlot, properties {horizontalAlignment=HAlign_Right, verticalAlignment=VAlign_Center}}
  }
 ]
}
