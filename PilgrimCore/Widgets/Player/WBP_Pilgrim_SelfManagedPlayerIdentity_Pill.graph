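# Self-managed wrapper widget, based on CommonUserWidget, that displays a player identity pill without needing external setup.
# Its only widget, and its root, is a WBP_Pilgrim_PlayerIdentity_Pill.
# On Construct it calls SetupPlayerIdentity on that pill with no player state supplied and forced visibility off, leaving identity resolution to the inner pill.
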WidgetBlueprint /PilgrimCore/Widgets/Player/WBP_Pilgrim_SelfManagedPlayerIdentity_Pill
graphs [EventGraph {nodes [
   Event Construct()
   > SetupPlayerIdentity!(
    execute
    self:WBP Pilgrim Player Identity Pill Object Reference=Get WBP_Pilgrim_PlayerIdentity_Pill() -> (WBP_Pilgrim_PlayerIdentity_Pill:WBP Pilgrim Player Identity Pill Object Reference)
    OptionalPlayerState:Player State Object Reference
    ForceVisible:bool=false
   ) -> (then)
  ]}]
variables []
widgets {
 info {
  inheritedWidgetCount=0
  parentClass=&CommonUserWidget
  rootWidgetClass=&/WBP_Pilgrim_PlayerIdentity_Pill.WBP_Pilgrim_PlayerIdentity_Pill_C
  widgetCount=1
 }
 widgets [WBP_Pilgrim_PlayerIdentity_Pill {widgetClassPath=&/WBP_Pilgrim_PlayerIdentity_Pill, bIsVariable=true, widget=WBP_Pilgrim_PlayerIdentity_Pill}]
}
